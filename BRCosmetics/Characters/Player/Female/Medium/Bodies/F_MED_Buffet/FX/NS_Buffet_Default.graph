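# Niagara cosmetic effect that adds pale blue and pink sprite sparkles to the F_MED_Buffet character body.
# Static emitters re-sample skeletal mesh location each update, while World emitters spawn per distance moved, drift through turbulence and stop spawning when User.IsSkydiving is set.
# GPU emitters cull particles using a body FX texture at sampled UVs; CPU 001 variants use another triangle sampling mode instead.
# Camera distance scales spawn density and sprite size.
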NiagaraSystem /BRCosmetics/Characters/Player/Female/Medium/Bodies/F_MED_Buffet/FX/NS_Buffet_Default
dependencies {
 usedDataInterfaces [
  &NiagaraDataInterfaceArrayColor
  /Script/Niagara.NiagaraDataInterfaceTexture
  &NiagaraDataInterfaceCurve
  /Script/Niagara.NiagaraDataInterfaceVectorField
  &NiagaraDataInterfaceSkeletalMesh
 ]
 usedDynamicInputs []
 usedModules [
  &self:SystemSpawnScript.SystemScriptSource.SystemScriptGraph.NiagaraNodeAssignment_0.SetVariables_#1
  &SystemState
 ]
}
emitters {
 Static_Sparkles {
  simTarget=GPUComputeSim
  emitterUpdateScript [
   &EmitterState()
   &SpawnRate(
    SpawnRate=&Multiply_Float(
     A=20
     B=&FN_LOD_Distance_Remap(Start Distance=0, End Distance=2000, Minimum Output=1.25, Maximum Output=0.125)
    )
    Use Spawn Probability=false
   )
  ]
  particleSpawnScript [
   &InitializeParticle(
    Color=&SelectLinearColorFromArray(
     Array Sampling Mode=::NewEnumerator1
     Direct Array Index=&RandomRangeInt(Minimum=0, Maximum=1)
     Color Selection Array=DataInterface() {propertyValues {ColorData [<rgba 0.75 0.8893 1 1>, <rgba 1 0.75 1 1>]}}
     Random Seed=?
    )
    Lifetime Max=1
    Lifetime Min=0.25
    Lifetime Mode=::NewEnumerator1
    Sprite Rotation Mode=::NewEnumerator2
    Sprite Size Mode=::NewEnumerator1
    Uniform Sprite Size Max=12
    Uniform Sprite Size Min=8
    UsePositionOffset=false
    Write Material Random=true
   )
   &CameraOffset(Camera Offset Amount=5)
  ]
  particleUpdateScript [
   &ParticleState()
   &SkeletalMeshLocation(
    MaskSpawnGroup=false
    Mesh Sampling Type=::NewEnumerator7
    OffsetAlongNormal=false
    OverrideRandomSeed=false
    Skeletal Mesh=$System.Skeletal Mesh
   )
   /Niagara/Modules/Update/Textures/SampleTexture(
    Texture=DataInterface() {propertyValues {
      Texture=/Game/Characters/Player/Female/Medium/Bodies/F_MED_Buffet/Textures/T_F_MED_Buffet_Body_FX
      TextureUserParameter {parameter {name=None, typeDefHandle {registeredTypeIndex=2}, varData []}}
     }}
    UV:Vector2f=$StackContext.SkeletalMeshLocation.SampledUV
   )
   &KillParticles(Kill Particles=&SetBoolByFloatComparison(
     A=&MakeFloatFromLinearColor(LinearColor=$StackContext.SampleTexture.SampledColor, Channel:ENiagaraLinearColor_Channels=::NewEnumerator1)
     B=0
     Comparison Type=::NewEnumerator7
    ))
   &DynamicMaterialParameters(
    Index 0 Param 1=&FloatFromCurve(FloatCurve=DataInterface() {propertyValues {Curve {keys [{arriveTangent=-3.2272, interpMode=RCIM_Cubic, leaveTangent=-3.2272, tangentMode=RCTM_User, time=0, value=1}, <at 1 0 RCIM_Cubic>]}}})
    Index 0 Param 2=&FloatFromCurve(FloatCurve=DataInterface() {propertyValues {Curve {keys [{arriveTangent=-12.4095, interpMode=RCIM_Cubic, leaveTangent=-12.4095, tangentMode=RCTM_User, time=0.5, value=2}, <at 1 0 RCIM_Cubic>]}}})
    Module.Param0WriteEnabled=true
    Module.Param1WriteEnabled=true
    Module.Param2WriteEnabled=true
    Module.Param3WriteEnabled=true
   )
   &ScaleSpriteSize(Uniform Curve Sprite Scale=DataInterface() {propertyValues {Curve {keys [
        {arriveTangent=14.9673, interpMode=RCIM_Cubic, leaveTangent=14.9673, tangentMode=RCTM_User, time=0, value=0}
        <at 0.2 1 RCIM_Cubic>
        {arriveTangent=-3.2414, interpMode=RCIM_Cubic, leaveTangent=-3.2414, tangentMode=RCTM_User, time=1, value=0}
       ]}}})
   &ScaleColor(
    Scale RGB=&VectorFromFloat(Value=&FloatFromCurve(FloatCurve=DataInterface() {propertyValues {Curve {keys [
          {arriveTangent=321.9435, interpMode=RCIM_Cubic, leaveTangent=321.9435, tangentMode=RCTM_User, time=0, value=0}
          {arriveTangentWeight=1.8705, interpMode=RCIM_Cubic, leaveTangentWeight=1.8705, tangentMode=RCTM_User, time=0.05, value=5}
          <at 0.1 1 RCIM_Cubic>
         ]}}}))
    ScaleA=true
    ScaleRGB=true
   )
   &ScaleSpriteSize(
    Scale Sprite Size Mode=::NewEnumerator0
    Uniform Scale Factor=&FN_LOD_Distance_Remap(Start Distance=0, End Distance=2000, Minimum Output=1, Maximum Output=2)
   ) {module=ScaleSpriteSize001}
  ]
  rendererClasses [&NiagaraSpriteRendererProperties]
  renderers [{rendererClass=&NiagaraSpriteRendererProperties}]
 }
 Static_Sparkles001 {
  emitterUpdateScript [
   &EmitterState()
   &SpawnRate(
    SpawnRate=&Multiply_Float(
     A=5
     B=&FN_LOD_Distance_Remap(Start Distance=0, End Distance=2000, Minimum Output=1.25, Maximum Output=0.125)
    )
    Use Spawn Probability=false
   )
  ]
  particleSpawnScript [
   &InitializeParticle(
    Color=&SelectLinearColorFromArray(
     Array Sampling Mode=::NewEnumerator1
     Direct Array Index=&RandomRangeInt(Minimum=0, Maximum=1)
     Color Selection Array=DataInterface() {propertyValues {ColorData [<rgba 0.75 0.8893 1 1>, <rgba 1 0.75 1 1>]}}
     Random Seed=?
    )
    Lifetime Max=1
    Lifetime Min=0.25
    Lifetime Mode=::NewEnumerator1
    Sprite Rotation Mode=::NewEnumerator2
    Sprite Size Mode=::NewEnumerator1
    Uniform Sprite Size Max=12
    Uniform Sprite Size Min=8
    UsePositionOffset=false
    Write Material Random=true
   )
   &CameraOffset(Camera Offset Amount=5)
  ]
  particleUpdateScript [
   &ParticleState()
   &SkeletalMeshLocation(
    MaskSpawnGroup=false
    Mesh Sampling Type=::NewEnumerator7
    OffsetAlongNormal=false
    OverrideRandomSeed=false
    Skeletal Mesh=$System.Skeletal Mesh
    Triangle Sampling Mode:ENiagaraTriangleSamplingMode=::NewEnumerator5
   )
   &DynamicMaterialParameters(
    Index 0 Param 1=&FloatFromCurve(FloatCurve=DataInterface() {propertyValues {Curve {keys [{arriveTangent=-3.2272, interpMode=RCIM_Cubic, leaveTangent=-3.2272, tangentMode=RCTM_User, time=0, value=1}, <at 1 0 RCIM_Cubic>]}}})
    Index 0 Param 2=&FloatFromCurve(FloatCurve=DataInterface() {propertyValues {Curve {keys [{arriveTangent=-12.4095, interpMode=RCIM_Cubic, leaveTangent=-12.4095, tangentMode=RCTM_User, time=0.5, value=2}, <at 1 0 RCIM_Cubic>]}}})
    Module.Param0WriteEnabled=true
    Module.Param1WriteEnabled=true
    Module.Param2WriteEnabled=true
    Module.Param3WriteEnabled=true
   )
   &ScaleSpriteSize(Uniform Curve Sprite Scale=DataInterface() {propertyValues {Curve {keys [
        {arriveTangent=14.9673, interpMode=RCIM_Cubic, leaveTangent=14.9673, tangentMode=RCTM_User, time=0, value=0}
        <at 0.2 1 RCIM_Cubic>
        {arriveTangent=-3.2414, interpMode=RCIM_Cubic, leaveTangent=-3.2414, tangentMode=RCTM_User, time=1, value=0}
       ]}}})
   &ScaleColor(
    Scale RGB=&VectorFromFloat(Value=&FloatFromCurve(FloatCurve=DataInterface() {propertyValues {Curve {keys [
          {arriveTangent=321.9435, interpMode=RCIM_Cubic, leaveTangent=321.9435, tangentMode=RCTM_User, time=0, value=0}
          {arriveTangentWeight=1.8705, interpMode=RCIM_Cubic, leaveTangentWeight=1.8705, tangentMode=RCTM_User, time=0.05, value=5}
          <at 0.1 1 RCIM_Cubic>
         ]}}}))
    ScaleA=true
    ScaleRGB=true
   )
   &ScaleSpriteSize(
    Scale Sprite Size Mode=::NewEnumerator0
    Uniform Scale Factor=&FN_LOD_Distance_Remap(Start Distance=0, End Distance=2000, Minimum Output=1, Maximum Output=2)
   ) {module=ScaleSpriteSize001}
  ]
  rendererClasses [&NiagaraSpriteRendererProperties]
  renderers [{rendererClass=&NiagaraSpriteRendererProperties}]
 }
 World_Sparkles {
  simTarget=GPUComputeSim
  emitterUpdateScript [
   &EmitterState()
   &SpawnPerUnit(
    Module.Use Max Movement Threshold=true
    Module.Use Movement Tolerance=false
    Spawn Probability:NiagaraFloat=&MakeCustomFloatFromBool(Bool=$User.IsSkydiving, True Float=0, False Float=0.5)
    Spawn Spacing:NiagaraFloat=&Multiply_Float(
     A=100
     B=&FN_LOD_Distance_Remap(Start Distance=0, End Distance=2000, Minimum Output=0.125, Maximum Output=1.25)
    )
    Use Spawn Probability=true
   )
  ]
  particleSpawnScript [
   &InitializeParticle(
    Color=&SelectLinearColorFromArray(
     Array Sampling Mode=::NewEnumerator1
     Direct Array Index=&RandomRangeInt(Minimum=0, Maximum=1)
     Color Selection Array=DataInterface() {propertyValues {ColorData [<rgba 0.75 0.8893 1 1>, <rgba 1 0.75 1 1>]}}
     Random Seed=?
    )
    Lifetime Max=1.5
    Lifetime Min=0.25
    Lifetime Mode=::NewEnumerator1
    Sprite Rotation Mode=::NewEnumerator2
    Sprite Size Mode=::NewEnumerator1
    Uniform Sprite Size Max=12
    Uniform Sprite Size Min=8
    UsePositionOffset=false
    Write Material Random=true
   )
   &SkeletalMeshLocation(
    MaskSpawnGroup=false
    Mesh Sampling Type=::NewEnumerator7
    OffsetAlongNormal=false
    OverrideRandomSeed=false
    Skeletal Mesh=$System.Skeletal Mesh
   )
   &SampleTexture(
    Texture=DataInterface() {propertyValues {
      Texture=&T_F_MED_Buffet_Body_FX
      TextureUserParameter {parameter {name=None, typeDefHandle {registeredTypeIndex=2}, varData []}}
     }}
    UV:Vector2f=$StackContext.SkeletalMeshLocation.SampledUV
   )
   &KillParticles(Kill Particles=&SetBoolByFloatComparison(
     A=&MakeFloatFromLinearColor(LinearColor=$StackContext.SampleTexture.SampledColor, Channel:ENiagaraLinearColor_Channels=::NewEnumerator1)
     B=0
     Comparison Type=::NewEnumerator7
    ))
   &CameraOffset(Camera Offset Amount=5)
   &AddVelocityFromPoint(Module.Use Falloff Distance=false, Velocity Strength=&RandomRangeFloat(Minimum=5, Maximum=10))
  ]
  particleUpdateScript [
   &ParticleState()
   &DynamicMaterialParameters(
    Index 0 Param 1=0.125
    Index 0 Param 2=&FloatFromCurve(FloatCurve=DataInterface() {propertyValues {Curve {keys [<at 0.5 1 RCIM_Cubic>, {arriveTangent=-3.4719, interpMode=RCIM_Cubic, leaveTangent=-3.4719, tangentMode=RCTM_User, time=1, value=0.5}]}}})
    Module.Param0WriteEnabled=true
    Module.Param1WriteEnabled=true
    Module.Param2WriteEnabled=true
    Module.Param3WriteEnabled=true
   )
   &ScaleSpriteSize(Uniform Curve Sprite Scale=DataInterface() {propertyValues {Curve {keys [
        {arriveTangent=14.9673, interpMode=RCIM_Cubic, leaveTangent=14.9673, tangentMode=RCTM_User, time=0, value=0}
        <at 0.2 1 RCIM_Cubic>
        {arriveTangent=-3.2414, interpMode=RCIM_Cubic, leaveTangent=-3.2414, tangentMode=RCTM_User, time=1, value=0}
       ]}}})
   &ScaleColor(
    Scale Alpha=0.5
    Scale RGB=&VectorFromFloat(Value=&FloatFromCurve(
      FloatCurve=DataInterface() {propertyValues {Curve {keys [
          {arriveTangent=23.2562, interpMode=RCIM_Cubic, leaveTangent=23.2562, tangentMode=RCTM_User, time=0, value=0}
          <at 0.1 1 RCIM_Cubic>
          <at 1 0 RCIM_Cubic>
         ]}}}
      Scale Curve=5
     ))
    ScaleA=true
    ScaleRGB=true
   )
   &SpriteRotationRate(Rotation Rate=360)
   &ScaleSpriteSize(
    Scale Sprite Size Mode=::NewEnumerator0
    Uniform Scale Factor=&FN_LOD_Distance_Remap(Start Distance=0, End Distance=2000, Minimum Output=1, Maximum Output=2)
   ) {module=ScaleSpriteSize001}
   /Niagara/VectorFields/SampleVectorField(VectorField=DataInterface() {propertyValues {Field=/Game/Effects/VectorFields/Noisy_Turbulence_103, bTileX=true, bTileY=true, bTileZ=true}})
   /Niagara/VectorFields/ApplyVectorField(Vector Field Intensity as Force=100, Vector Field Intensity as Velocity=1)
   &AccelerationForce(Acceleration=&RandomRangeVector(Minimum=<0 0 -75>, Maximum=<0 0 75>))
   &SolveForcesAndVelocity(Clamp Velocity=false, Limit Acceleration=false)
  ]
  rendererClasses [&NiagaraSpriteRendererProperties]
  renderers [{rendererClass=&NiagaraSpriteRendererProperties}]
 }
 World_Sparkles001 {
  emitterUpdateScript [
   &EmitterState()
   &SpawnPerUnit(
    Module.Use Max Movement Threshold=true
    Module.Use Movement Tolerance=false
    Spawn Probability:NiagaraFloat=&MakeCustomFloatFromBool(Bool=$User.IsSkydiving, True Float=0, False Float=0.5)
    Spawn Spacing:NiagaraFloat=&Multiply_Float(
     A=50
     B=&FN_LOD_Distance_Remap(Start Distance=0, End Distance=2000, Minimum Output=0.125, Maximum Output=1.25)
    )
    Use Spawn Probability=true
   )
  ]
  particleSpawnScript [
   &InitializeParticle(
    Color=&SelectLinearColorFromArray(
     Array Sampling Mode=::NewEnumerator1
     Direct Array Index=&RandomRangeInt(Minimum=0, Maximum=1)
     Color Selection Array=DataInterface() {propertyValues {ColorData [<rgba 0.75 0.8893 1 1>, <rgba 1 0.75 1 1>]}}
     Random Seed=?
    )
    Lifetime Max=1.5
    Lifetime Min=0.25
    Lifetime Mode=::NewEnumerator1
    Sprite Rotation Mode=::NewEnumerator2
    Sprite Size Mode=::NewEnumerator1
    Uniform Sprite Size Max=12
    Uniform Sprite Size Min=8
    UsePositionOffset=false
    Write Material Random=true
   )
   &SkeletalMeshLocation(
    MaskSpawnGroup=false
    Mesh Sampling Type=::NewEnumerator7
    OffsetAlongNormal=false
    OverrideRandomSeed=false
    Skeletal Mesh=$System.Skeletal Mesh
    Triangle Sampling Mode:ENiagaraTriangleSamplingMode=::NewEnumerator5
   )
   &CameraOffset(Camera Offset Amount=5)
   &AddVelocityFromPoint(Module.Use Falloff Distance=false, Velocity Strength=&RandomRangeFloat(Minimum=5, Maximum=10))
  ]
  particleUpdateScript [
   &ParticleState()
   &DynamicMaterialParameters(
    Index 0 Param 1=0.125
    Index 0 Param 2=&FloatFromCurve(FloatCurve=DataInterface() {propertyValues {Curve {keys [<at 0.5 1 RCIM_Cubic>, {arriveTangent=-3.4719, interpMode=RCIM_Cubic, leaveTangent=-3.4719, tangentMode=RCTM_User, time=1, value=0.5}]}}})
    Module.Param0WriteEnabled=true
    Module.Param1WriteEnabled=true
    Module.Param2WriteEnabled=true
    Module.Param3WriteEnabled=true
   )
   &ScaleSpriteSize(Uniform Curve Sprite Scale=DataInterface() {propertyValues {Curve {keys [
        {arriveTangent=14.9673, interpMode=RCIM_Cubic, leaveTangent=14.9673, tangentMode=RCTM_User, time=0, value=0}
        <at 0.2 1 RCIM_Cubic>
        {arriveTangent=-3.2414, interpMode=RCIM_Cubic, leaveTangent=-3.2414, tangentMode=RCTM_User, time=1, value=0}
       ]}}})
   &ScaleColor(
    Scale Alpha=0.5
    Scale RGB=&VectorFromFloat(Value=&FloatFromCurve(
      FloatCurve=DataInterface() {propertyValues {Curve {keys [
          {arriveTangent=23.2562, interpMode=RCIM_Cubic, leaveTangent=23.2562, tangentMode=RCTM_User, time=0, value=0}
          <at 0.1 1 RCIM_Cubic>
          <at 1 0 RCIM_Cubic>
         ]}}}
      Scale Curve=5
     ))
    ScaleA=true
    ScaleRGB=true
   )
   &SpriteRotationRate(Rotation Rate=360)
   &ScaleSpriteSize(
    Scale Sprite Size Mode=::NewEnumerator0
    Uniform Scale Factor=&FN_LOD_Distance_Remap(Start Distance=0, End Distance=2000, Minimum Output=1, Maximum Output=2)
   ) {module=ScaleSpriteSize001}
   &SampleVectorField(VectorField=DataInterface() {propertyValues {Field=&Noisy_Turbulence_103, bTileX=true, bTileY=true, bTileZ=true}})
   &ApplyVectorField(Vector Field Intensity as Force=100, Vector Field Intensity as Velocity=1)
   &AccelerationForce(Acceleration=&RandomRangeVector(Minimum=<0 0 -75>, Maximum=<0 0 75>))
   &SolveForcesAndVelocity(Clamp Velocity=false, Limit Acceleration=false)
  ]
  rendererClasses [&NiagaraSpriteRendererProperties]
  renderers [{rendererClass=&NiagaraSpriteRendererProperties}]
 }
}
summary {
 systemName=NS_Buffet_Default
 emitters [
  {emitterName=Static_Sparkles, simTarget=GPUComputeSim, rendererClasses [&NiagaraSpriteRendererProperties]}
  {emitterName=World_Sparkles, simTarget=GPUComputeSim, rendererClasses [&NiagaraSpriteRendererProperties]}
  {emitterName=Static_Sparkles001, rendererClasses [&NiagaraSpriteRendererProperties]}
  {emitterName=World_Sparkles001, rendererClasses [&NiagaraSpriteRendererProperties]}
 ]
 userVariables [User.IsSkydiving {type=NiagaraBool}]
}
