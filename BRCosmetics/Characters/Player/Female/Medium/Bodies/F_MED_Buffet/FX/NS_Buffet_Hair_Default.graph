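# Niagara cosmetic effect for the F_MED_Buffet character's hair, emitting pink sparkle sprites around skeletal mesh bones or sockets.
# Spawn groups index the bone/socket plus per-group color and cylinder shape.
# Static emitters re-sample bones each update to stay attached, while World emitters spawn by movement, drift through turbulence and stop spawning during User.IsSkydiving.
# Each exists as GPU and default-sim (CPU) copies, with camera-distance remaps scaling spawn density and sprite size.
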NiagaraSystem /BRCosmetics/Characters/Player/Female/Medium/Bodies/F_MED_Buffet/FX/NS_Buffet_Hair_Default
dependencies {
 usedDataInterfaces [
  &NiagaraDataInterfaceArrayColor
  &NiagaraDataInterfaceArrayFloat
  &NiagaraDataInterfaceCurve
  &NiagaraDataInterfaceArrayFloat3
  /Script/Niagara.NiagaraDataInterfaceVectorField
  &NiagaraDataInterfaceSkeletalMesh
 ]
 usedDynamicInputs []
 usedModules [
  &self:SystemSpawnScript.SystemScriptSource.SystemScriptGraph.NiagaraNodeAssignment_0.SetVariables_#1
  &SystemState
 ]
}
emitters {
 Static_Sparkles {
  simTarget=GPUComputeSim
  emitterUpdateScript [
   &EmitterState()
   &SpawnRate(
    SpawnRate=&Multiply_Float(
     A=2
     B=&FN_LOD_Distance_Remap(Start Distance=0, End Distance=2000, Minimum Output=1.25, Maximum Output=0.125)
    )
    Use Spawn Probability=false
   )
   &SpawnRate(
    SpawnGroup=1
    SpawnRate=&Multiply_Float(
     A=3
     B=&FN_LOD_Distance_Remap(Start Distance=0, End Distance=2000, Minimum Output=1.25, Maximum Output=0.125)
    )
    Use Spawn Probability=false
   ) {module=SpawnRate001}
   &SpawnRate(
    SpawnGroup=2
    SpawnRate=&Multiply_Float(
     A=3
     B=&FN_LOD_Distance_Remap(Start Distance=0, End Distance=2000, Minimum Output=1.25, Maximum Output=0.125)
    )
    Use Spawn Probability=false
   ) {module=SpawnRate002}
   &SpawnRate(
    SpawnGroup=3
    SpawnRate=&Multiply_Float(
     A=3
     B=&FN_LOD_Distance_Remap(Start Distance=0, End Distance=2000, Minimum Output=1.25, Maximum Output=0.125)
    )
    Use Spawn Probability=false
   ) {module=SpawnRate003}
   &SpawnRate(
    SpawnGroup=4
    SpawnRate=&Multiply_Float(
     A=3
     B=&FN_LOD_Distance_Remap(Start Distance=0, End Distance=2000, Minimum Output=1.25, Maximum Output=0.125)
    )
    Use Spawn Probability=false
   ) {module=SpawnRate004}
  ]
  particleSpawnScript [
   &InitializeParticle(
    Color=&SelectLinearColorFromArray(
     Array Sampling Mode=::NewEnumerator1
     Direct Array Index=&SpawnGroup()
     Color Selection Array=DataInterface() {propertyValues {ColorData [<rgba 1 0.75 1 1>, <rgba 1 0.75 1 1>, <rgba 1 0.2 1 1>, <rgba 1 0.2 1 1>, <rgba 1 0.2 1 1>, <rgba 1 0.2 1 1>]}}
     Random Seed=?
    )
    Lifetime Max=1
    Lifetime Min=0.25
    Lifetime Mode=::NewEnumerator1
    Sprite Rotation Mode=::NewEnumerator2
    Sprite Size Mode=::NewEnumerator1
    Uniform Sprite Size Max=12
    Uniform Sprite Size Min=8
    UsePositionOffset=false
    Write Material Random=true
   )
   &CameraOffset(Camera Offset Amount=5)
   &CylinderLocation(
    Apply To Particle Position:ENiagara_AttributeSamplingApplyOutput=::NewEnumerator1
    Cylinder Height:NiagaraFloat=&SelectFloatFromArray(
     Direct Array Index=&SpawnGroup()
     Float Selection Array=DataInterface() {propertyValues {FloatData [1,30,20,30,20,20]}}
    )
    Cylinder Origin:NiagaraPosition=/Niagara/DynamicInputs/Helpers/SimulationPosition()
    Cylinder Radius=&SelectFloatFromArray(
     Direct Array Index=&SpawnGroup()
     Float Selection Array=DataInterface() {propertyValues {FloatData [2,3,5,5,4,0.125]}}
    )
    MaskSpawnGroup=false
    Offset=&SelectVectorFromArray(
     Direct Array Index=&SpawnGroup()
     Vector Selection Array=DataInterface() {propertyValues {FloatData [<0 0 0>, <0 0 0>, <0 0 0>, <0 0 0>, <0 0 0>, <4 0 0>]}}
    )
    Offset Mode:ENiagara_OffsetMode=::NewEnumerator0
    Override Seed=false
    Roll:NiagaraFloat=&SelectFloatFromArray(
     Direct Array Index=&SpawnGroup()
     Float Selection Array=DataInterface() {propertyValues {FloatData [0,0,0,0,0,0]}}
    )
    Rotation Coordinate Space:ENiagaraCoordinateSpace=::Simulation
    Rotation Mode:ENiagara_RotationMode=::NewEnumerator2
    UseOwnerScale=true
   )
  ]
  particleUpdateScript [
   &ParticleState()
   &SkeletalMeshLocation(
    Bone / Socket Index=&SpawnGroup()
    Bone Sampling Mode=::NewEnumerator4
    MaskSpawnGroup=false
    Sampled Position Offset=$Particles.CylinderLocation.CylinderPosition
    Sampled Position Offset Coordinate Space:ENiagaraSkelMeshTransforms=::NewEnumerator0
    Skeletal Mesh=$System.Skeletal Mesh
   )
   &DynamicMaterialParameters(
    Index 0 Param 1=&FloatFromCurve(FloatCurve=DataInterface() {propertyValues {Curve {keys [{arriveTangent=-3.2272, interpMode=RCIM_Cubic, leaveTangent=-3.2272, tangentMode=RCTM_User, time=0, value=1}, <at 1 0 RCIM_Cubic>]}}})
    Index 0 Param 2=&FloatFromCurve(FloatCurve=DataInterface() {propertyValues {Curve {keys [{arriveTangent=-12.4095, interpMode=RCIM_Cubic, leaveTangent=-12.4095, tangentMode=RCTM_User, time=0.5, value=2}, <at 1 0 RCIM_Cubic>]}}})
    Module.Param0WriteEnabled=true
    Module.Param1WriteEnabled=true
    Module.Param2WriteEnabled=true
    Module.Param3WriteEnabled=true
   )
   &ScaleSpriteSize(Uniform Curve Sprite Scale=DataInterface() {propertyValues {Curve {keys [
        {arriveTangent=14.9673, interpMode=RCIM_Cubic, leaveTangent=14.9673, tangentMode=RCTM_User, time=0, value=0}
        <at 0.2 1 RCIM_Cubic>
        {arriveTangent=-3.2414, interpMode=RCIM_Cubic, leaveTangent=-3.2414, tangentMode=RCTM_User, time=1, value=0}
       ]}}})
   &ScaleColor(
    Scale RGB=&VectorFromFloat(Value=&FloatFromCurve(FloatCurve=DataInterface() {propertyValues {Curve {keys [
          {arriveTangent=321.9435, interpMode=RCIM_Cubic, leaveTangent=321.9435, tangentMode=RCTM_User, time=0, value=0}
          {arriveTangentWeight=1.8705, interpMode=RCIM_Cubic, leaveTangentWeight=1.8705, tangentMode=RCTM_User, time=0.05, value=5}
          <at 0.1 1 RCIM_Cubic>
         ]}}}))
    ScaleA=true
    ScaleRGB=true
   )
   &ScaleSpriteSize(
    Scale Sprite Size Mode=::NewEnumerator0
    Uniform Scale Factor=&FN_LOD_Distance_Remap(Start Distance=0, End Distance=2000, Minimum Output=1, Maximum Output=2)
   ) {module=ScaleSpriteSize001}
  ]
  rendererClasses [&NiagaraSpriteRendererProperties]
  renderers [{rendererClass=&NiagaraSpriteRendererProperties}]
 }
 Static_Sparkles001 {
  emitterUpdateScript [
   &EmitterState()
   &SpawnRate(
    SpawnRate=&Multiply_Float(
     A=2
     B=&FN_LOD_Distance_Remap(Start Distance=0, End Distance=2000, Minimum Output=1.25, Maximum Output=0.125)
    )
    Use Spawn Probability=false
   )
   &SpawnRate(
    SpawnGroup=1
    SpawnRate=&Multiply_Float(
     A=3
     B=&FN_LOD_Distance_Remap(Start Distance=0, End Distance=2000, Minimum Output=1.25, Maximum Output=0.125)
    )
    Use Spawn Probability=false
   ) {module=SpawnRate001}
   &SpawnRate(
    SpawnGroup=2
    SpawnRate=&Multiply_Float(
     A=3
     B=&FN_LOD_Distance_Remap(Start Distance=0, End Distance=2000, Minimum Output=1.25, Maximum Output=0.125)
    )
    Use Spawn Probability=false
   ) {module=SpawnRate002}
   &SpawnRate(
    SpawnGroup=3
    SpawnRate=&Multiply_Float(
     A=3
     B=&FN_LOD_Distance_Remap(Start Distance=0, End Distance=2000, Minimum Output=1.25, Maximum Output=0.125)
    )
    Use Spawn Probability=false
   ) {module=SpawnRate003}
   &SpawnRate(
    SpawnGroup=4
    SpawnRate=&Multiply_Float(
     A=3
     B=&FN_LOD_Distance_Remap(Start Distance=0, End Distance=2000, Minimum Output=1.25, Maximum Output=0.125)
    )
    Use Spawn Probability=false
   ) {module=SpawnRate004}
  ]
  particleSpawnScript [
   &InitializeParticle(
    Color=&SelectLinearColorFromArray(
     Array Sampling Mode=::NewEnumerator1
     Direct Array Index=&SpawnGroup()
     Color Selection Array=DataInterface() {propertyValues {ColorData [<rgba 1 0.75 1 1>, <rgba 1 0.75 1 1>, <rgba 1 0.2 1 1>, <rgba 1 0.2 1 1>, <rgba 1 0.2 1 1>, <rgba 1 0.2 1 1>]}}
     Random Seed=?
    )
    Lifetime Max=1
    Lifetime Min=0.25
    Lifetime Mode=::NewEnumerator1
    Sprite Rotation Mode=::NewEnumerator2
    Sprite Size Mode=::NewEnumerator1
    Uniform Sprite Size Max=12
    Uniform Sprite Size Min=8
    UsePositionOffset=false
    Write Material Random=true
   )
   &CameraOffset(Camera Offset Amount=5)
   &CylinderLocation(
    Apply To Particle Position:ENiagara_AttributeSamplingApplyOutput=::NewEnumerator1
    Cylinder Height:NiagaraFloat=&SelectFloatFromArray(
     Direct Array Index=&SpawnGroup()
     Float Selection Array=DataInterface() {propertyValues {FloatData [1,30,20,30,20,20]}}
    )
    Cylinder Origin:NiagaraPosition=&SimulationPosition()
    Cylinder Radius=&SelectFloatFromArray(
     Direct Array Index=&SpawnGroup()
     Float Selection Array=DataInterface() {propertyValues {FloatData [2,3,5,5,4,0.125]}}
    )
    MaskSpawnGroup=false
    Offset=&SelectVectorFromArray(
     Direct Array Index=&SpawnGroup()
     Vector Selection Array=DataInterface() {propertyValues {FloatData [<0 0 0>, <0 0 0>, <0 0 0>, <0 0 0>, <0 0 0>, <4 0 0>]}}
    )
    Offset Mode:ENiagara_OffsetMode=::NewEnumerator0
    Override Seed=false
    Roll:NiagaraFloat=&SelectFloatFromArray(
     Direct Array Index=&SpawnGroup()
     Float Selection Array=DataInterface() {propertyValues {FloatData [0,0,0,0,0,0]}}
    )
    Rotation Coordinate Space:ENiagaraCoordinateSpace=::Simulation
    Rotation Mode:ENiagara_RotationMode=::NewEnumerator2
    UseOwnerScale=true
   )
  ]
  particleUpdateScript [
   &ParticleState()
   &SkeletalMeshLocation(
    Bone / Socket Index=&SpawnGroup()
    Bone Sampling Mode=::NewEnumerator4
    MaskSpawnGroup=false
    Sampled Position Offset=$Particles.CylinderLocation.CylinderPosition
    Sampled Position Offset Coordinate Space:ENiagaraSkelMeshTransforms=::NewEnumerator0
    Skeletal Mesh=$System.Skeletal Mesh
   )
   &DynamicMaterialParameters(
    Index 0 Param 1=&FloatFromCurve(FloatCurve=DataInterface() {propertyValues {Curve {keys [{arriveTangent=-3.2272, interpMode=RCIM_Cubic, leaveTangent=-3.2272, tangentMode=RCTM_User, time=0, value=1}, <at 1 0 RCIM_Cubic>]}}})
    Index 0 Param 2=&FloatFromCurve(FloatCurve=DataInterface() {propertyValues {Curve {keys [{arriveTangent=-12.4095, interpMode=RCIM_Cubic, leaveTangent=-12.4095, tangentMode=RCTM_User, time=0.5, value=2}, <at 1 0 RCIM_Cubic>]}}})
    Module.Param0WriteEnabled=true
    Module.Param1WriteEnabled=true
    Module.Param2WriteEnabled=true
    Module.Param3WriteEnabled=true
   )
   &ScaleSpriteSize(Uniform Curve Sprite Scale=DataInterface() {propertyValues {Curve {keys [
        {arriveTangent=14.9673, interpMode=RCIM_Cubic, leaveTangent=14.9673, tangentMode=RCTM_User, time=0, value=0}
        <at 0.2 1 RCIM_Cubic>
        {arriveTangent=-3.2414, interpMode=RCIM_Cubic, leaveTangent=-3.2414, tangentMode=RCTM_User, time=1, value=0}
       ]}}})
   &ScaleColor(
    Scale RGB=&VectorFromFloat(Value=&FloatFromCurve(FloatCurve=DataInterface() {propertyValues {Curve {keys [
          {arriveTangent=321.9435, interpMode=RCIM_Cubic, leaveTangent=321.9435, tangentMode=RCTM_User, time=0, value=0}
          {arriveTangentWeight=1.8705, interpMode=RCIM_Cubic, leaveTangentWeight=1.8705, tangentMode=RCTM_User, time=0.05, value=5}
          <at 0.1 1 RCIM_Cubic>
         ]}}}))
    ScaleA=true
    ScaleRGB=true
   )
   &ScaleSpriteSize(
    Scale Sprite Size Mode=::NewEnumerator0
    Uniform Scale Factor=&FN_LOD_Distance_Remap(Start Distance=0, End Distance=2000, Minimum Output=1, Maximum Output=2)
   ) {module=ScaleSpriteSize001}
  ]
  rendererClasses [&NiagaraSpriteRendererProperties]
  renderers [{rendererClass=&NiagaraSpriteRendererProperties}]
 }
 World_Sparklies {
  simTarget=GPUComputeSim
  emitterUpdateScript [
   &EmitterState()
   &SpawnPerUnit(
    Module.Use Max Movement Threshold=true
    Module.Use Movement Tolerance=false
    Spawn Probability:NiagaraFloat=&MakeCustomFloatFromBool(Bool=$User.IsSkydiving, True Float=0, False Float=0.5)
    Spawn Spacing:NiagaraFloat=&Multiply_Float(
     A=400
     B=&FN_LOD_Distance_Remap(Start Distance=0, End Distance=2000, Minimum Output=0.125, Maximum Output=1.25)
    )
    Use Spawn Probability=true
   )
   &SpawnPerUnit(
    Module.Use Max Movement Threshold=true
    Module.Use Movement Tolerance=false
    Spawn Group=1
    Spawn Probability:NiagaraFloat=&MakeCustomFloatFromBool(Bool=$User.IsSkydiving, True Float=0, False Float=0.5)
    Spawn Spacing:NiagaraFloat=&Multiply_Float(
     A=200
     B=&FN_LOD_Distance_Remap(Start Distance=0, End Distance=2000, Minimum Output=0.125, Maximum Output=1.25)
    )
    Use Spawn Probability=true
   ) {module=SpawnPerUnit001}
   &SpawnPerUnit(
    Module.Use Max Movement Threshold=true
    Module.Use Movement Tolerance=false
    Spawn Group=2
    Spawn Probability:NiagaraFloat=&MakeCustomFloatFromBool(Bool=$User.IsSkydiving, True Float=0, False Float=0.5)
    Spawn Spacing:NiagaraFloat=&Multiply_Float(
     A=200
     B=&FN_LOD_Distance_Remap(Start Distance=0, End Distance=2000, Minimum Output=0.125, Maximum Output=1.25)
    )
    Use Spawn Probability=true
   ) {module=SpawnPerUnit002}
   &SpawnPerUnit(
    Module.Use Max Movement Threshold=true
    Module.Use Movement Tolerance=false
    Spawn Group=3
    Spawn Probability:NiagaraFloat=&MakeCustomFloatFromBool(Bool=$User.IsSkydiving, True Float=0, False Float=0.5)
    Spawn Spacing:NiagaraFloat=&Multiply_Float(
     A=200
     B=&FN_LOD_Distance_Remap(Start Distance=0, End Distance=2000, Minimum Output=0.125, Maximum Output=1.25)
    )
    Use Spawn Probability=true
   ) {module=SpawnPerUnit003}
   &SpawnPerUnit(
    Module.Use Max Movement Threshold=true
    Module.Use Movement Tolerance=false
    Spawn Group=4
    Spawn Probability:NiagaraFloat=&MakeCustomFloatFromBool(Bool=$User.IsSkydiving, True Float=0, False Float=0.5)
    Spawn Spacing:NiagaraFloat=&Multiply_Float(
     A=200
     B=&FN_LOD_Distance_Remap(Start Distance=0, End Distance=2000, Minimum Output=0.125, Maximum Output=1.25)
    )
    Use Spawn Probability=true
   ) {module=SpawnPerUnit004}
  ]
  particleSpawnScript [
   &InitializeParticle(
    Color=&SelectLinearColorFromArray(
     Array Sampling Mode=::NewEnumerator1
     Direct Array Index=&SpawnGroup()
     Color Selection Array=DataInterface() {propertyValues {ColorData [<rgba 1 0.75 1 1>, <rgba 1 0.75 1 1>, <rgba 1 0.5 1 1>, <rgba 1 0.5 1 1>, <rgba 1 0.5 1 1>, <rgba 1 0.5 1 1>]}}
     Random Seed=?
    )
    Lifetime Max=1.5
    Lifetime Min=0.25
    Lifetime Mode=::NewEnumerator1
    Sprite Rotation Mode=::NewEnumerator2
    Sprite Size Mode=::NewEnumerator1
    Uniform Sprite Size Max=12
    Uniform Sprite Size Min=8
    UsePositionOffset=false
    Write Material Random=true
   )
   &SkeletalMeshLocation(
    Bone / Socket Index=&SpawnGroup()
    Bone Sampling Mode=::NewEnumerator4
    MaskSpawnGroup=false
    Sampled Position Offset Coordinate Space:ENiagaraSkelMeshTransforms=::NewEnumerator0
    Skeletal Mesh=$System.Skeletal Mesh
   )
   &CylinderLocation(
    Apply To Particle Position:ENiagara_AttributeSamplingApplyOutput=::NewEnumerator1
    Cylinder Height:NiagaraFloat=&SelectFloatFromArray(
     Direct Array Index=&SpawnGroup()
     Float Selection Array=DataInterface() {propertyValues {FloatData [1,30,20,30,20,20]}}
    )
    Cylinder Origin:NiagaraPosition=&SimulationPosition()
    Cylinder Radius=&SelectFloatFromArray(
     Direct Array Index=&SpawnGroup()
     Float Selection Array=DataInterface() {propertyValues {FloatData [2,3,5,5,4,0.125]}}
    )
    MaskSpawnGroup=false
    Offset=&SelectVectorFromArray(
     Direct Array Index=&SpawnGroup()
     Vector Selection Array=DataInterface() {propertyValues {FloatData [<0 0 0>, <0 0 0>, <0 0 0>, <0 0 0>, <0 0 0>, <4 0 0>]}}
    )
    Offset Mode:ENiagara_OffsetMode=::NewEnumerator0
    Override Seed=false
    Roll:NiagaraFloat=&SelectFloatFromArray(
     Direct Array Index=&SpawnGroup()
     Float Selection Array=DataInterface() {propertyValues {FloatData [0,0,0,0,0,0]}}
    )
    Rotation Coordinate Space:ENiagaraCoordinateSpace=::Simulation
    Rotation Mode:ENiagara_RotationMode=::NewEnumerator2
    UseOwnerScale=true
   )
   &CameraOffset(Camera Offset Amount=5)
   &AddVelocityFromPoint(Module.Use Falloff Distance=false, Velocity Strength=&RandomRangeFloat(Minimum=5, Maximum=10))
  ]
  particleUpdateScript [
   &ParticleState()
   &DynamicMaterialParameters(
    Index 0 Param 1=&FloatFromCurve(FloatCurve=DataInterface() {propertyValues {Curve {keys [{arriveTangent=-3.2272, interpMode=RCIM_Cubic, leaveTangent=-3.2272, tangentMode=RCTM_User, time=0, value=1}, <at 1 0 RCIM_Cubic>]}}})
    Index 0 Param 2=&FloatFromCurve(FloatCurve=DataInterface() {propertyValues {Curve {keys [{arriveTangent=-12.4095, interpMode=RCIM_Cubic, leaveTangent=-12.4095, tangentMode=RCTM_User, time=0.5, value=2}, <at 1 0 RCIM_Cubic>]}}})
    Module.Param0WriteEnabled=true
    Module.Param1WriteEnabled=true
    Module.Param2WriteEnabled=true
    Module.Param3WriteEnabled=true
   )
   &ScaleSpriteSize(Uniform Curve Sprite Scale=DataInterface() {propertyValues {Curve {keys [
        {arriveTangent=14.9673, interpMode=RCIM_Cubic, leaveTangent=14.9673, tangentMode=RCTM_User, time=0, value=0}
        <at 0.2 1 RCIM_Cubic>
        {arriveTangent=-3.2414, interpMode=RCIM_Cubic, leaveTangent=-3.2414, tangentMode=RCTM_User, time=1, value=0}
       ]}}})
   &ScaleColor(
    Scale Alpha=0.25
    Scale RGB=&VectorFromFloat(Value=&FloatFromCurve(
      FloatCurve=DataInterface() {propertyValues {Curve {keys [
          {arriveTangent=23.2562, interpMode=RCIM_Cubic, leaveTangent=23.2562, tangentMode=RCTM_User, time=0, value=0}
          <at 0.1 1 RCIM_Cubic>
          <at 1 0 RCIM_Cubic>
         ]}}}
      Scale Curve=5
     ))
    ScaleA=true
    ScaleRGB=true
   )
   &SpriteRotationRate(Rotation Rate=360)
   &ScaleSpriteSize(
    Scale Sprite Size Mode=::NewEnumerator0
    Uniform Scale Factor=&FN_LOD_Distance_Remap(Start Distance=0, End Distance=2000, Minimum Output=1, Maximum Output=2)
   ) {module=ScaleSpriteSize001}
   /Niagara/VectorFields/SampleVectorField(VectorField=DataInterface() {propertyValues {Field=/Game/Effects/VectorFields/Noisy_Turbulence_103, bTileX=true, bTileY=true, bTileZ=true}})
   /Niagara/VectorFields/ApplyVectorField(Vector Field Intensity as Force=100, Vector Field Intensity as Velocity=1)
   &AccelerationForce(Acceleration=&RandomRangeVector(Minimum=<0 0 -75>, Maximum=<0 0 75>))
   &InheritSourceMovement(
    Applied Position Scale=&VectorFromFloat(Value=&FloatFromCurve(FloatCurve=DataInterface() {propertyValues {Curve {keys [{arriveTangent=-4.3234, interpMode=RCIM_Cubic, leaveTangent=-4.3234, tangentMode=RCTM_User, time=0, value=1}, <at 0.5 0 RCIM_Cubic>]}}}))
    ApplyForce=false
    ApplyPosition=true
    ApplyPositionFalloff=false
   )
   &SolveForcesAndVelocity(Clamp Velocity=false, Limit Acceleration=false)
  ]
  rendererClasses [&NiagaraSpriteRendererProperties]
  renderers [{rendererClass=&NiagaraSpriteRendererProperties}]
 }
 World_Sparklies001 {
  emitterUpdateScript [
   &EmitterState()
   &SpawnPerUnit(
    Module.Use Max Movement Threshold=true
    Module.Use Movement Tolerance=false
    Spawn Probability:NiagaraFloat=&MakeCustomFloatFromBool(Bool=$User.IsSkydiving, True Float=0, False Float=0.5)
    Spawn Spacing:NiagaraFloat=&Multiply_Float(
     A=400
     B=&FN_LOD_Distance_Remap(Start Distance=0, End Distance=2000, Minimum Output=0.125, Maximum Output=1.25)
    )
    Use Spawn Probability=true
   )
   &SpawnPerUnit(
    Module.Use Max Movement Threshold=true
    Module.Use Movement Tolerance=false
    Spawn Group=1
    Spawn Probability:NiagaraFloat=&MakeCustomFloatFromBool(Bool=$User.IsSkydiving, True Float=0, False Float=0.5)
    Spawn Spacing:NiagaraFloat=&Multiply_Float(
     A=200
     B=&FN_LOD_Distance_Remap(Start Distance=0, End Distance=2000, Minimum Output=0.125, Maximum Output=1.25)
    )
    Use Spawn Probability=true
   ) {module=SpawnPerUnit001}
   &SpawnPerUnit(
    Module.Use Max Movement Threshold=true
    Module.Use Movement Tolerance=false
    Spawn Group=2
    Spawn Probability:NiagaraFloat=&MakeCustomFloatFromBool(Bool=$User.IsSkydiving, True Float=0, False Float=0.5)
    Spawn Spacing:NiagaraFloat=&Multiply_Float(
     A=200
     B=&FN_LOD_Distance_Remap(Start Distance=0, End Distance=2000, Minimum Output=0.125, Maximum Output=1.25)
    )
    Use Spawn Probability=true
   ) {module=SpawnPerUnit002}
   &SpawnPerUnit(
    Module.Use Max Movement Threshold=true
    Module.Use Movement Tolerance=false
    Spawn Group=3
    Spawn Probability:NiagaraFloat=&MakeCustomFloatFromBool(Bool=$User.IsSkydiving, True Float=0, False Float=0.5)
    Spawn Spacing:NiagaraFloat=&Multiply_Float(
     A=200
     B=&FN_LOD_Distance_Remap(Start Distance=0, End Distance=2000, Minimum Output=0.125, Maximum Output=1.25)
    )
    Use Spawn Probability=true
   ) {module=SpawnPerUnit003}
   &SpawnPerUnit(
    Module.Use Max Movement Threshold=true
    Module.Use Movement Tolerance=false
    Spawn Group=4
    Spawn Probability:NiagaraFloat=&MakeCustomFloatFromBool(Bool=$User.IsSkydiving, True Float=0, False Float=0.5)
    Spawn Spacing:NiagaraFloat=&Multiply_Float(
     A=200
     B=&FN_LOD_Distance_Remap(Start Distance=0, End Distance=2000, Minimum Output=0.125, Maximum Output=1.25)
    )
    Use Spawn Probability=true
   ) {module=SpawnPerUnit004}
  ]
  particleSpawnScript [
   &InitializeParticle(
    Color=&SelectLinearColorFromArray(
     Array Sampling Mode=::NewEnumerator1
     Direct Array Index=&SpawnGroup()
     Color Selection Array=DataInterface() {propertyValues {ColorData [<rgba 1 0.75 1 1>, <rgba 1 0.75 1 1>, <rgba 1 0.5 1 1>, <rgba 1 0.5 1 1>, <rgba 1 0.5 1 1>, <rgba 1 0.5 1 1>]}}
     Random Seed=?
    )
    Lifetime Max=1.5
    Lifetime Min=0.25
    Lifetime Mode=::NewEnumerator1
    Sprite Rotation Mode=::NewEnumerator2
    Sprite Size Mode=::NewEnumerator1
    Uniform Sprite Size Max=12
    Uniform Sprite Size Min=8
    UsePositionOffset=false
    Write Material Random=true
   )
   &SkeletalMeshLocation(
    Bone / Socket Index=&SpawnGroup()
    Bone Sampling Mode=::NewEnumerator4
    MaskSpawnGroup=false
    Sampled Position Offset Coordinate Space:ENiagaraSkelMeshTransforms=::NewEnumerator0
    Skeletal Mesh=$System.Skeletal Mesh
   )
   &CylinderLocation(
    Apply To Particle Position:ENiagara_AttributeSamplingApplyOutput=::NewEnumerator1
    Cylinder Height:NiagaraFloat=&SelectFloatFromArray(
     Direct Array Index=&SpawnGroup()
     Float Selection Array=DataInterface() {propertyValues {FloatData [1,30,20,30,20,20]}}
    )
    Cylinder Origin:NiagaraPosition=&SimulationPosition()
    Cylinder Radius=&SelectFloatFromArray(
     Direct Array Index=&SpawnGroup()
     Float Selection Array=DataInterface() {propertyValues {FloatData [2,3,5,5,4,0.125]}}
    )
    MaskSpawnGroup=false
    Offset=&SelectVectorFromArray(
     Direct Array Index=&SpawnGroup()
     Vector Selection Array=DataInterface() {propertyValues {FloatData [<0 0 0>, <0 0 0>, <0 0 0>, <0 0 0>, <0 0 0>, <4 0 0>]}}
    )
    Offset Mode:ENiagara_OffsetMode=::NewEnumerator0
    Override Seed=false
    Roll:NiagaraFloat=&SelectFloatFromArray(
     Direct Array Index=&SpawnGroup()
     Float Selection Array=DataInterface() {propertyValues {FloatData [0,0,0,0,0,0]}}
    )
    Rotation Coordinate Space:ENiagaraCoordinateSpace=::Simulation
    Rotation Mode:ENiagara_RotationMode=::NewEnumerator2
    UseOwnerScale=true
   )
   &CameraOffset(Camera Offset Amount=5)
   &AddVelocityFromPoint(Module.Use Falloff Distance=false, Velocity Strength=&RandomRangeFloat(Minimum=5, Maximum=10))
  ]
  particleUpdateScript [
   &ParticleState()
   &DynamicMaterialParameters(
    Index 0 Param 1=&FloatFromCurve(FloatCurve=DataInterface() {propertyValues {Curve {keys [{arriveTangent=-3.2272, interpMode=RCIM_Cubic, leaveTangent=-3.2272, tangentMode=RCTM_User, time=0, value=1}, <at 1 0 RCIM_Cubic>]}}})
    Index 0 Param 2=&FloatFromCurve(FloatCurve=DataInterface() {propertyValues {Curve {keys [{arriveTangent=-12.4095, interpMode=RCIM_Cubic, leaveTangent=-12.4095, tangentMode=RCTM_User, time=0.5, value=2}, <at 1 0 RCIM_Cubic>]}}})
    Module.Param0WriteEnabled=true
    Module.Param1WriteEnabled=true
    Module.Param2WriteEnabled=true
    Module.Param3WriteEnabled=true
   )
   &ScaleSpriteSize(Uniform Curve Sprite Scale=DataInterface() {propertyValues {Curve {keys [
        {arriveTangent=14.9673, interpMode=RCIM_Cubic, leaveTangent=14.9673, tangentMode=RCTM_User, time=0, value=0}
        <at 0.2 1 RCIM_Cubic>
        {arriveTangent=-3.2414, interpMode=RCIM_Cubic, leaveTangent=-3.2414, tangentMode=RCTM_User, time=1, value=0}
       ]}}})
   &ScaleColor(
    Scale Alpha=0.25
    Scale RGB=&VectorFromFloat(Value=&FloatFromCurve(
      FloatCurve=DataInterface() {propertyValues {Curve {keys [
          {arriveTangent=23.2562, interpMode=RCIM_Cubic, leaveTangent=23.2562, tangentMode=RCTM_User, time=0, value=0}
          <at 0.1 1 RCIM_Cubic>
          <at 1 0 RCIM_Cubic>
         ]}}}
      Scale Curve=5
     ))
    ScaleA=true
    ScaleRGB=true
   )
   &SpriteRotationRate(Rotation Rate=360)
   &ScaleSpriteSize(
    Scale Sprite Size Mode=::NewEnumerator0
    Uniform Scale Factor=&FN_LOD_Distance_Remap(Start Distance=0, End Distance=2000, Minimum Output=1, Maximum Output=2)
   ) {module=ScaleSpriteSize001}
   &SampleVectorField(VectorField=DataInterface() {propertyValues {Field=&Noisy_Turbulence_103, bTileX=true, bTileY=true, bTileZ=true}})
   &ApplyVectorField(Vector Field Intensity as Force=100, Vector Field Intensity as Velocity=1)
   &AccelerationForce(Acceleration=&RandomRangeVector(Minimum=<0 0 -75>, Maximum=<0 0 75>))
   &InheritSourceMovement(
    Applied Position Scale=&VectorFromFloat(Value=&FloatFromCurve(FloatCurve=DataInterface() {propertyValues {Curve {keys [{arriveTangent=-4.3234, interpMode=RCIM_Cubic, leaveTangent=-4.3234, tangentMode=RCTM_User, time=0, value=1}, <at 0.5 0 RCIM_Cubic>]}}}))
    ApplyForce=false
    ApplyPosition=true
    ApplyPositionFalloff=false
   )
   &SolveForcesAndVelocity(Clamp Velocity=false, Limit Acceleration=false)
  ]
  rendererClasses [&NiagaraSpriteRendererProperties]
  renderers [{rendererClass=&NiagaraSpriteRendererProperties}]
 }
}
summary {
 systemName=NS_Buffet_Hair_Default
 emitters [
  {emitterName=Static_Sparkles, simTarget=GPUComputeSim, rendererClasses [&NiagaraSpriteRendererProperties]}
  {emitterName=World_Sparklies, simTarget=GPUComputeSim, rendererClasses [&NiagaraSpriteRendererProperties]}
  {emitterName=Static_Sparkles001, rendererClasses [&NiagaraSpriteRendererProperties]}
  {emitterName=World_Sparklies001, rendererClasses [&NiagaraSpriteRendererProperties]}
 ]
 userVariables [User.IsSkydiving {type=NiagaraBool}]
}
