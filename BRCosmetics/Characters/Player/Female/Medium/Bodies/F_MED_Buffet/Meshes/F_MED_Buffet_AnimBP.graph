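# Animation Blueprint for the F_MED_Buffet body cosmetic, adding muscle deformation and skirt motion to the copied parent-mesh pose.
# A Control Rig takes pose-space deformer values, then the simspace bone gets rotation from a deprecated update function and pelvis-constrained position in the front end.
# Skirt bones blend pelvis/thigh rotation and add thigh-driven offsets before simspace-based RigidBody physics, weighted by crouch.
# Center front and back bones then average their left/right neighbors.
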AnimBlueprint /BRCosmetics/Characters/Player/Female/Medium/Bodies/F_MED_Buffet/Meshes/F_MED_Buffet_AnimBP
target_skeleton=&/F_MED_Buffet_Skeleton
graphs [
 AnimGraph {nodes [
   n1 = "Constraint - dyn_skirt_fr_a_base_l (RR)"!(
    ComponentPose=Local To Component(LocalPose=Use cached pose 'deformRigPose'())
    Alpha=1
    ConstraintWeights_0=0.1
    ConstraintWeights_1:float=0.9
   ) -> (Pose) {
    alphaBoolBlend {blendOption=Linear}
    boneToModify {boneName=dyn_skirt_fr_a_base_l}
    constraintSetup [{transformType=Rotation, targetBone {boneName=pelvis}}, {transformType=Rotation, targetBone {boneName=thigh_l}}]
    constraintWeights [1,1]
   }
   n2 = "Constraint - dyn_skirt_side_a_base_l (RR)"!(ComponentPose=@n1, Alpha=1, ConstraintWeights_0=0.4, ConstraintWeights_1:float=0.6) -> (Pose) {
    alphaBoolBlend {blendOption=Linear}
    boneToModify {boneName=dyn_skirt_side_a_base_l}
    constraintSetup [{transformType=Rotation, targetBone {boneName=pelvis}}, {transformType=Rotation, targetBone {boneName=thigh_l}}]
    constraintWeights [1,1]
   }
   n3 = "Constraint - dyn_skirt_bk_a_base_l (RR)"!(ComponentPose=@n2, Alpha=1, ConstraintWeights_0=0.9, ConstraintWeights_1:float=0.1) -> (Pose) {
    alphaBoolBlend {blendOption=Linear}
    boneToModify {boneName=dyn_skirt_bk_a_base_l}
    constraintSetup [{transformType=Rotation, targetBone {boneName=pelvis}}, {transformType=Rotation, targetBone {boneName=thigh_l}}]
    constraintWeights [1,1]
   }
   n4 = "Constraint - dyn_skirt_fr_a_base_r (RR)"!(ComponentPose=@n3, Alpha=1, ConstraintWeights_0=0.1, ConstraintWeights_1:float=0.9) -> (Pose) {
    alphaBoolBlend {blendOption=Linear}
    boneToModify {boneName=dyn_skirt_fr_a_base_r}
    constraintSetup [{transformType=Rotation, targetBone {boneName=pelvis}}, {transformType=Rotation, targetBone {boneName=thigh_r}}]
    constraintWeights [1,1]
   }
   n5 = "Constraint - dyn_skirt_side_a_base_r (RR)"!(ComponentPose=@n4, Alpha=1, ConstraintWeights_0=0.4, ConstraintWeights_1:float=0.6) -> (Pose) {
    alphaBoolBlend {blendOption=Linear}
    boneToModify {boneName=dyn_skirt_side_a_base_r}
    constraintSetup [{transformType=Rotation, targetBone {boneName=pelvis}}, {transformType=Rotation, targetBone {boneName=thigh_r}}]
    constraintWeights [1,1]
   }
   n6 = "Constraint - dyn_skirt_bk_a_base_r (RR)"!(ComponentPose=@n5, Alpha=1, ConstraintWeights_0=0.9, ConstraintWeights_1:float=0.1) -> (Pose) {
    alphaBoolBlend {blendOption=Linear}
    boneToModify {boneName=dyn_skirt_bk_a_base_r}
    constraintSetup [{transformType=Rotation, targetBone {boneName=pelvis}}, {transformType=Rotation, targetBone {boneName=thigh_r}}]
    constraintWeights [1,1]
   }
   n7 = Break Buffet Pose Space Deformer Input!(BuffetPoseSpaceDeformerInput:"Buffet Pose Space Deformer Input Structure (by ref)"=Get ThighRotL() -> (ThighRotL:Buffet Pose Space Deformer Input Structure)) -> (DeformationValue:float=~) {is=title K2Node_BreakStruct}
   n8 = "Transform (Modify) Bone - Bone: dyn_skirt_fr_a_base_l"(ComponentPose=@n6, Alpha=@n7, Translation="2,0,6", Rotation="5,5,0", Scale="1,1,1") {
    rotationMode=BMM_Additive
    rotationSpace=BCS_BoneSpace
    scaleMode=BMM_Additive
    scaleSpace=BCS_BoneSpace
    translationMode=BMM_Additive
    translationSpace=BCS_BoneSpace
    alphaBoolBlend {blendOption=Linear}
    alphaScaleBiasClamp {bMapRange=true, inRange {max=0, min=-100}, outRange {max=0, min=1}}
    boneToModify {boneName=dyn_skirt_fr_a_base_l}
    rotation {pitch=5, yaw=5}
    translation {x=2, z=6}
   }
   n9 = "Transform (Modify) Bone - Bone: dyn_skirt_side_a_base_l"(ComponentPose=@n8, Alpha=@n7, Translation="0,0,4", Rotation="0,0,0", Scale="1,1,1") {
    rotationMode=BMM_Additive
    rotationSpace=BCS_BoneSpace
    scaleMode=BMM_Additive
    scaleSpace=BCS_BoneSpace
    translationMode=BMM_Additive
    translationSpace=BCS_BoneSpace
    alphaBoolBlend {blendOption=Linear}
    alphaScaleBiasClamp {bMapRange=true, inRange {max=0, min=-100}, outRange {max=0, min=1}}
    boneToModify {boneName=dyn_skirt_side_a_base_l}
    translation {z=4}
   }
   n10 = "Transform (Modify) Bone - Bone: dyn_skirt_bk_a_base_l"(ComponentPose=@n9, Alpha=@n7, Translation="0,0,2", Rotation="0,0,0", Scale="1,1,1") {
    rotationMode=BMM_Additive
    rotationSpace=BCS_BoneSpace
    scaleMode=BMM_Additive
    scaleSpace=BCS_BoneSpace
    translationMode=BMM_Additive
    translationSpace=BCS_BoneSpace
    alphaBoolBlend {blendOption=Linear}
    alphaScaleBiasClamp {bMapRange=true, inRange {max=0, min=-100}, outRange {max=0, min=1}}
    boneToModify {boneName=dyn_skirt_bk_a_base_l}
    translation {z=2}
   }
   n11 = Break Buffet Pose Space Deformer Input(BuffetPoseSpaceDeformerInput=Get ThighRotR() -> (ThighRotR:Buffet Pose Space Deformer Input Structure)) -> (DeformationValue=~)
   n12 = "Transform (Modify) Bone - Bone: dyn_skirt_fr_a_base_r"(ComponentPose=@n10, Alpha=@n11, Translation="-2,0,-5", Rotation="5,5,0", Scale="1,1,1") {
    rotationMode=BMM_Additive
    rotationSpace=BCS_BoneSpace
    scaleMode=BMM_Additive
    scaleSpace=BCS_BoneSpace
    translationMode=BMM_Additive
    translationSpace=BCS_BoneSpace
    alphaBoolBlend {blendOption=Linear}
    alphaScaleBiasClamp {bMapRange=true, inRange {max=0, min=-100}, outRange {max=0, min=1}}
    boneToModify {boneName=dyn_skirt_fr_a_base_r}
    rotation {pitch=5, yaw=5}
    translation {x=-2, z=-5}
   }
   n13 = "Transform (Modify) Bone - Bone: dyn_skirt_side_a_base_r"(ComponentPose=@n12, Alpha=@n11, Translation="0,0,-4", Rotation="0,0,0", Scale="1,1,1") {
    rotationMode=BMM_Additive
    rotationSpace=BCS_BoneSpace
    scaleMode=BMM_Additive
    scaleSpace=BCS_BoneSpace
    translationMode=BMM_Additive
    translationSpace=BCS_BoneSpace
    alphaBoolBlend {blendOption=Linear}
    alphaScaleBiasClamp {bMapRange=true, inRange {max=0, min=-100}, outRange {max=0, min=1}}
    boneToModify {boneName=dyn_skirt_side_a_base_r}
    translation {z=-4}
   }
   n14 = "Transform (Modify) Bone - Bone: dyn_skirt_bk_a_base_r"(ComponentPose=@n13, Alpha=@n11, Translation="0,0,-2", Rotation="0,0,0", Scale="1,1,1") {
    rotationMode=BMM_Additive
    rotationSpace=BCS_BoneSpace
    scaleMode=BMM_Additive
    scaleSpace=BCS_BoneSpace
    translationMode=BMM_Additive
    translationSpace=BCS_BoneSpace
    alphaBoolBlend {blendOption=Linear}
    alphaScaleBiasClamp {bMapRange=true, inRange {max=0, min=-100}, outRange {max=0, min=1}}
    boneToModify {boneName=dyn_skirt_bk_a_base_r}
    translation {z=-2}
   }
   n15 = RigidBody!(
    ComponentPose=@n14
    Alpha=Get crouchingAlpha() -> (crouchingAlpha:float)
    OverrideWorldGravity:Vector=Get gravityOverride() -> (gravityOverride:Vector)
    ExternalForce
   ) -> (Pose) {
    bClampLinearTranslationLimitToRefPose=true
    bOverrideWorldGravity=true
    overridePhysicsAsset=&/F_MED_Buffet_Physics
    simulationSpace=BaseBoneSpace
    alphaBoolBlend {blendOption=Linear}
    alphaScaleBiasClamp {bMapRange=true, outRange {max=0.5, min=1}}
    baseBoneRef {boneName=simspace}
    componentAppliedLinearAccClamp=<3000 3000 3000>
    componentLinearAccScale=<0.25 0.25 0.25>
    componentLinearVelScale=<0.25 0.25 0.25>
    overrideWorldGravity {z=-980}
   }
   n16 = "Constraint - dyn_skirt_fr_a_base (PP)"!(ComponentPose=@n15, Alpha=1, ConstraintWeights_0=0.5, ConstraintWeights_1:float=0.5) -> (Pose) {
    alphaBoolBlend {blendOption=Linear}
    boneToModify {boneName=dyn_skirt_fr_a_base}
    constraintSetup [{targetBone {boneName=dyn_skirt_fr_a_l}}, {targetBone {boneName=dyn_skirt_fr_a_r}}]
    constraintWeights [1,1]
   }
   n17 = "Constraint - dyn_skirt_bk_a_base (PP)"!(ComponentPose=@n16, Alpha=1, ConstraintWeights_0=0.5, ConstraintWeights_1:float=0.5) -> (Pose) {
    alphaBoolBlend {blendOption=Linear}
    boneToModify {boneName=dyn_skirt_bk_a_base}
    constraintSetup [{targetBone {boneName=dyn_skirt_bk_a_l}}, {targetBone {boneName=dyn_skirt_bk_a_r}}]
    constraintWeights [1,1]
   }
   Output Pose(Result=Component To Local(ComponentPose=@n17)) {layerGroup=None}
   n18 = Break Pose Deformation Query Results Arms!(PoseDeformationQueryResults_Arms:"Pose Deformation Query Results Arms Structure (by ref)"=Break Pose Deformation Query Arms!(PoseDeformationQuery_Arms:"Pose Deformation Query Arms Structure (by ref)"=Get Arms() -> (Arms:Pose Deformation Query Arms Structure)) -> (QueryResults:Pose Deformation Query Results Arms Structure) {is=title K2Node_BreakStruct}) -> (
    ArmUpL:float=~
    ArmUpR:float=~
    ArmForwardL:float=~
    ArmForwardR:float=~
    ArmBackL:float=~
    ArmBackR:float=~
    ArmForwardUpL:float=~
    ArmForwardUpR:float=~
    ArmForwardDownL:float=~
    ArmForwardDownR:float=~
   ) {is=title K2Node_BreakStruct}
   n19 = Control Rig!(
    Source=Copy Pose From Mesh() {bUseAttachedParent=true}
    Alpha
    thigh_rot_l:double=Break Buffet Pose Space Deformer Input(BuffetPoseSpaceDeformerInput=Get ThighRotL()) -> (DeformationValue=~)
    elbow_rot_l:double=Break Buffet Pose Space Deformer Input(BuffetPoseSpaceDeformerInput=Get ElbowRotL() -> (ElbowRotL:Buffet Pose Space Deformer Input Structure)) -> (DeformationValue=~)
    elbow_rot_r:double=Break Buffet Pose Space Deformer Input(BuffetPoseSpaceDeformerInput=Get ElbowRotR() -> (ElbowRotR:Buffet Pose Space Deformer Input Structure)) -> (DeformationValue=~)
    thigh_rot_r:double=Break Buffet Pose Space Deformer Input(BuffetPoseSpaceDeformerInput=Get ThighRotR()) -> (DeformationValue=~)
    knee_rot_l:double=Break Buffet Pose Space Deformer Input(BuffetPoseSpaceDeformerInput=Get KneeRotL() -> (KneeRotL:Buffet Pose Space Deformer Input Structure)) -> (DeformationValue=~)
    knee_rot_r:double=Break Buffet Pose Space Deformer Input(BuffetPoseSpaceDeformerInput=Get KneeRotR() -> (KneeRotR:Buffet Pose Space Deformer Input Structure)) -> (DeformationValue=~)
    psd_arm_up_alpha_l:double=@n18
    psd_arm_up_alpha_r:double=@n18.ArmUpR
    psd_arm_fwd_alpha_l:double=@n18.ArmForwardL
    psd_arm_fwd_alpha_r:double=@n18.ArmForwardR
    psd_arm_bk_alpha_l:double=@n18.ArmBackL
    psd_arm_bk_alpha_r:double=@n18.ArmBackR
    psd_arm_fwd_up_alpha_l:double=@n18.ArmForwardUpL
    psd_arm_fwd_up_alpha_r:double=@n18.ArmForwardUpR
    psd_arm_fwd_dn_alpha_l:double=@n18.ArmForwardDownL
    psd_arm_fwd_dn_alpha_r:double=@n18.ArmForwardDownR
   ) -> (Pose) {
    alphaBoolBlend {blendOption=Linear}
    controlRigAssetReference {blueprintRigClass=/BRCosmetics/Characters/Player/Female/Medium/Bodies/F_MED_Buffet_Shard/Meshes/F_MED_Muscle_Volume_ControlRig.F_MED_Muscle_Volume_ControlRig_C}
   }
   n20 = "Transform (Modify) Bone - Bone: simspace"(
    ComponentPose=Local To Component(LocalPose=@n19)
    Alpha=1
    Translation="0,0,0"
    Rotation=Get sim_space_rotation_$() -> (sim_space_rotation_$:Rotator)
    Scale="1,1,1"
   ) {
    rotationMode=BMM_Replace
    rotationSpace=BCS_WorldSpace
    translationSpace=BCS_BoneSpace
    alphaBoolBlend {blendOption=Linear}
    alphaScaleBiasClamp {bInterpResult=true, interpSpeedDecreasing=2, interpSpeedIncreasing=2}
    boneToModify {boneName=simspace}
   }
   n21 = "Constraint - simspace (P)"!(ComponentPose=@n20, bAlphaBoolEnabled:bool=Get bIsInFrontEnd(), ConstraintWeights_0=1) -> (Pose) {
    alphaInputType=Bool
    alphaBoolBlend {blendOption=Linear}
    boneToModify {boneName=simspace}
    constraintSetup [{targetBone {boneName=pelvis}}]
    constraintWeights [1]
   }
   Save cached pose 'deformRigPose'(Pose=Component To Local(ComponentPose=@n21)) {cacheName=deformRigPose}
  ]}
 EventGraph {nodes [
   Event BlueprintUpdateAnimation()
   > n1 = ComplexSimSpaceFunction DEPRECATED!(
    execute
    CharacterPartAnimInstance:Custom Character Part Anim Instance Object Reference=Self-Reference()
    InterpSpeedMax:double=9
    PawnSpeedMax:double=20
    RInterpErrorTolerance:double=1
    DeltaTime:double=0
    DebugDraw:bool=false
    sim_space_rotation_in:Rotator=Get sim_space_rotation_$()
    inComponentSpaceIn:bool=Get InComponentSpace_$() -> (InComponentSpace_$:bool)
    inComponentSpaceLastFrameIn:bool=Get InComponentSpaceLastFrame_$() -> (InComponentSpaceLastFrame_$:bool)
    lockedOnSimSpaceIn:bool=Get LockedOnSimSpace_$() -> (LockedOnSimSpace_$:bool)
    InterpSpeedIn:double=Get InterpSpeed_$() -> (InterpSpeed_$:double)
   ) -> (
    then
    sim_space_rotation:Rotator
    inComponentSpaceOut:bool
    inComponentSpaceLastFrameOut:bool
    lockedOnSimSpaceOut:bool
    InterpSpeedOut:double
   )
   > Set sim_space_rotation_$(sim_space_rotation_$:Rotator=@n1) -> (Output_Get:Rotator)
   > Set InComponentSpace_$(InComponentSpace_$:bool=@n1.inComponentSpaceOut) -> (Output_Get:bool)
   > Set InComponentSpaceLastFrame_$(InComponentSpaceLastFrame_$:bool=@n1.inComponentSpaceLastFrameOut) -> (Output_Get:bool)
   > Set LockedOnSimSpace_$(LockedOnSimSpace_$:bool=@n1.lockedOnSimSpaceOut) -> (Output_Get:bool)
   > Set InterpSpeed_$(InterpSpeed_$:double=@n1.InterpSpeedOut) -> (Output_Get:double)
  ]}
]
variables [
 InComponentSpace_$ {type=boolean}
 InComponentSpaceLastFrame_$ {type=boolean}
 InterpSpeed_$ {type=number}
 LockedOnSimSpace_$ {type=boolean}
 sim_space_rotation_$ {type=Rotator}
]
