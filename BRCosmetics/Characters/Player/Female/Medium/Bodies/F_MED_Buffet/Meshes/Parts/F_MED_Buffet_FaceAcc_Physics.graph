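# Physics asset for the face accessory of the Buffet medium female player body, covering a dynamic hair chain and the body it can collide with.
# Six hair capsules (dyn_hair_a to f) form a chain from a small base body on neck_01, with moderate swing limits and tight twist limits.
# Capsules on the pelvis, spine, neck, arms and thighs give body volumes, linked by free-rotation constraints, as is the hair base to the neck.
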PhysicsAsset /BRCosmetics/Characters/Player/Female/Medium/Bodies/F_MED_Buffet/Meshes/Parts/F_MED_Buffet_FaceAcc_Physics
bodies {
 dyn_hair_a [{shapeType=Capsule, length=8, radius=5.5, center=<10 0 0>, rotation=<rot 90 0 0>}]
 dyn_hair_a_base [{shapeType=Capsule, length=1, radius=1}]
 dyn_hair_b [{shapeType=Capsule, length=7.75, radius=5.25, center=<9 0 0>, rotation=<rot 90 0 0>}]
 dyn_hair_c [{shapeType=Capsule, length=9.75, radius=5.5, center=<11 0 0>, rotation=<rot 90 0 0>}]
 dyn_hair_d [{shapeType=Capsule, length=8.75, radius=5.5, center=<10 0 0>, rotation=<rot 90 0 0>}]
 dyn_hair_e [{shapeType=Capsule, length=14.75, radius=4.5, center=<12 0 0>, rotation=<rot 90 0 0>}]
 dyn_hair_f [{shapeType=Capsule, length=11.25, radius=3.75, center=<10 0 0>, rotation=<rot 90 0 0>}]
 lowerarm_l [{shapeType=Capsule, length=30.5, radius=6.5, center=<15.7336 -0.0001 0>, rotation=<rot 90 0 0>}]
 lowerarm_r [{shapeType=Capsule, length=24.6794, radius=12.3206, center=<-15.7336 0.0001 0>, rotation=<rot 90 0 180>}]
 neck_01 [{shapeType=Capsule, length=14.005, radius=5.755, shapeName=neck_01_capsule, center=<2.2687 0 0>, rotation=<rot 0 -89.9999 0>}]
 pelvis [{shapeType=Capsule, length=25.0613, radius=20.0522, center=<-3.742 0.8329 0>}]
 spine_02 [{shapeType=Capsule, length=12.505, radius=9.255, shapeName=spine_02_capsule, rotation=<rot 0 -90 0>}]
 spine_05 [{shapeType=Capsule, length=12.25, radius=12.5, center=<-0.8242 3.3872 -0.0002>}]
 thigh_l [{shapeType=Capsule, length=29.3074, radius=23.7812, center=<-24.665 0 -0.0002>, rotation=<rot 90 0 0>}]
 thigh_r [{shapeType=Capsule, length=29.3074, radius=23.7812, center=<24.665 0 0.0002>, rotation=<rot 90 0 180>}]
 upperarm_l [{shapeType=Capsule, length=18.75, radius=6.75, center=<12.954 -0.0001 0>, rotation=<rot 90 0 0>}]
 upperarm_r [{shapeType=Capsule, length=12.9294, radius=12.5706, center=<-12.954 0.0001 0>, rotation=<rot 90 0 180>}]
}
constraints [
 {bone1Name=dyn_hair_a, bone2Name=dyn_hair_a_base, swing1LimitDegrees=30, swing2LimitDegrees=20, twistLimitDegrees=5}
 {bone1Name=dyn_hair_b, bone2Name=dyn_hair_a, swing1LimitDegrees=30, swing2LimitDegrees=20, twistLimitDegrees=5}
 {bone1Name=dyn_hair_c, bone2Name=dyn_hair_b, swing1LimitDegrees=30, swing2LimitDegrees=30, twistLimitDegrees=5}
 {bone1Name=dyn_hair_d, bone2Name=dyn_hair_c, swing1LimitDegrees=30, swing2LimitDegrees=30, twistLimitDegrees=5}
 {bone1Name=dyn_hair_e, bone2Name=dyn_hair_d, swing1LimitDegrees=30, swing2LimitDegrees=30, twistLimitDegrees=5}
 {bone1Name=dyn_hair_f, bone2Name=dyn_hair_e, swing1LimitDegrees=30, swing2LimitDegrees=30, twistLimitDegrees=5}
 {
  bone1Name=dyn_hair_a_base
  bone2Name=neck_01
  swing1LimitDegrees=45
  swing1Motion=Free
  swing2LimitDegrees=45
  swing2Motion=Free
  twistLimitDegrees=45
  twistMotion=Free
 }
 {
  bone1Name=spine_05
  bone2Name=spine_02
  swing1LimitDegrees=45
  swing1Motion=Free
  swing2LimitDegrees=45
  swing2Motion=Free
  twistLimitDegrees=45
  twistMotion=Free
 }
 {
  bone1Name=upperarm_l
  bone2Name=spine_05
  swing1LimitDegrees=45
  swing1Motion=Free
  swing2LimitDegrees=45
  swing2Motion=Free
  twistLimitDegrees=45
  twistMotion=Free
 }
 {
  bone1Name=lowerarm_l
  bone2Name=upperarm_l
  swing1LimitDegrees=45
  swing1Motion=Free
  swing2LimitDegrees=45
  swing2Motion=Free
  twistLimitDegrees=45
  twistMotion=Free
 }
 {
  bone1Name=upperarm_r
  bone2Name=spine_05
  swing1LimitDegrees=45
  swing1Motion=Free
  swing2LimitDegrees=45
  swing2Motion=Free
  twistLimitDegrees=45
  twistMotion=Free
 }
 {
  bone1Name=lowerarm_r
  bone2Name=upperarm_r
  swing1LimitDegrees=45
  swing1Motion=Free
  swing2LimitDegrees=45
  swing2Motion=Free
  twistLimitDegrees=45
  twistMotion=Free
 }
 {
  bone1Name=thigh_l
  bone2Name=pelvis
  swing1LimitDegrees=45
  swing1Motion=Free
  swing2LimitDegrees=45
  swing2Motion=Free
  twistLimitDegrees=45
  twistMotion=Free
 }
 {
  bone1Name=thigh_r
  bone2Name=pelvis
  swing1LimitDegrees=45
  swing1Motion=Free
  swing2LimitDegrees=45
  swing2Motion=Free
  twistLimitDegrees=45
  twistMotion=Free
 }
 {
  bone1Name=spine_02
  bone2Name=pelvis
  swing1LimitDegrees=45
  swing1Motion=Free
  swing2LimitDegrees=45
  swing2Motion=Free
  twistLimitDegrees=45
  twistMotion=Free
 }
 {
  bone1Name=neck_01
  bone2Name=spine_05
  swing1LimitDegrees=45
  swing1Motion=Free
  swing2LimitDegrees=45
  swing2Motion=Free
  twistLimitDegrees=45
  twistMotion=Free
 }
]
